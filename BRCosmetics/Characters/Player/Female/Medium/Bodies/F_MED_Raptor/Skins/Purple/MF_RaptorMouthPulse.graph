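# Material function for the purple female Raptor skin that adds an animated, flashing mouth pulse to base color and emissive, gated by a static switch.
# A procedural wavy line combines an arcsine-of-sine triangle wave, blue-noise jitter, stutter and a smile bend, with optional behind-teeth masking.
# Layered sines drive flicker, combined with a mouth gradient and edges from offset samples of the FX mask's blue channel.
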MaterialFunction /BRCosmetics/Characters/Player/Female/Medium/Bodies/F_MED_Raptor/Skins/Purple/MF_RaptorMouthPulse
graph {nodes [
  n1 = FunctionInput() {inputType=FunctionInput_MaterialAttributes}
  n2 = GetMaterialAttributes(None=@n1) {attributeGetTypes [#1, #2]}
  n3 = NamedRerouteUsage()
  n4 = NamedRerouteUsage()
  n5 = LinearInterpolate(A=@n2.BaseColor, B=@n3, Alpha=@n4)
  n6 = LinearInterpolate(A=@n2.EmissiveColor, B=@n3, Alpha=@n4)
  FunctionOutput(None=StaticSwitchParameter(
    True=SetMaterialAttributes!(MaterialAttributes=@n2.MaterialAttributes, BaseColor=@n5, EmissiveColor=@n6) {attributeSetTypes [#1, #2]}
    False=@n1
   ) {parameterName=use RaptorMouthPulse})
  n7 = TextureSampleParameter2D() {
   parameterName=MaskFX
   samplerType=SAMPLERTYPE_Masks
   texture=/BRCosmetics/Characters/Player/Female/Medium/Bodies/F_MED_Raptor/Skins/Purple/Textures/T_F_MED_Raptor_Purple_Mask_FX
  }
  NamedRerouteDeclaration(None=@n7.B) {name=Texture Mask, nodeColor=<rgba 0.0078 1 0 1>}
  NamedRerouteDeclaration(None=SphereMask(A=TextureCoordinate(), B=Constant2Vector() {g=0.5, r=0.153}, Radius=Constant() {r=0.06}) {attenuationRadius=0.07, hardnessPercent=50}) {name=Mouth Area Mask, nodeColor=<rgba 0.7344 0 1 1>}
  NamedRerouteDeclaration(None=VectorParameter() {parameterName=Flash Color, defaultValue=<rgba 0.0907 0.7525 0.7917 1>}.RGB) {name=Emissive Color, nodeColor=<rgba 0.2422 1 0 1>}
  n8 = Add(A=ScalarParameter() {defaultValue=6.5, group=Raptor Mouth, parameterName=Amplitude}, B=Multiply(A=NamedRerouteUsage()) {constB=0.5})
  n9 = MaterialFunctionCall() {materialFunction=&Pi}
  n10 = Multiply(A=Time(), B=ScalarParameter() {group=Raptor Mouth, parameterName=NoisePanSpeed})
  n11 = NamedRerouteUsage()
  n12 = Multiply(
   A=Divide(A=Multiply(A=@n9.Result) {constA=2, constB=2}) {constB=0.5}
   B=Multiply(A=ScalarParameter() {defaultValue=0.8, group=Raptor Mouth, parameterName=TriangleWave_Wavelength}, B=Add(A=@n10, B=@n11))
  )
  n13 = Multiply(A=Divide(A=Multiply(A=@n8) {constB=2}, B=@n9.Result), B=ArcsineFast!(None=Sine(None=@n12)))
  n14 = Time()
  n15 = ScalarParameter() {defaultValue=0.6, group=Raptor Mouth, parameterName=Noise_HighFrequency_Speed}
  n16 = Add(A=@n11, B=NamedRerouteDeclaration(None=Multiply(A=@n14, B=@n15) {constB=0.33}) {name=HighFrequency Noise, nodeColor=<rgba 0.3359 0 1 1>})
  n17 = TextureSample(UVs=Multiply(A=@n16, B=ScalarParameter() {defaultValue=0.05, group=Raptor Mouth, parameterName=Noise_HighFrequency_Wavelength})) {
   samplerSource=SSM_Wrap_WorldGroupSettings
   samplerType=SAMPLERTYPE_LinearGrayscale
   texture=/BRCosmetics/Global/Textures/Default/Noise/T_BlueNoise
  }
  n18 = Multiply(
   A=ConstantBiasScale(None=@n17.G) {bias=-0.5, scale=2}
   B=ScalarParameter() {defaultValue=1.5, group=Raptor Mouth, parameterName=Noise_HighFrequency_Intensity}
  ) {constB=0.25}
  n19 = TextureSample(UVs=Multiply(A=@n16, B=ScalarParameter() {defaultValue=0.03, group=Raptor Mouth, parameterName=PolarityShift_Wavelength})) {
   samplerSource=SSM_Wrap_WorldGroupSettings
   samplerType=SAMPLERTYPE_LinearGrayscale
   texture=&T_BlueNoise
  }
  n20 = NamedRerouteDeclaration(None=@n19.G) {name=Wavelength, nodeColor=<rgba 0 1 0.5313 1>}
  n21 = NamedRerouteUsage()
  n22 = Tangent!(None=Time())
  n23 = ComponentMask(None=Panner(
    Coordinate=Add(A=@n20, B=@n21)
    Time=Clamp(None=@n22)
    Speed=ScalarParameter() {defaultValue=0.3, group=Raptor Mouth, parameterName=Stutter Intensity}
   ) {bFractionalPart=true}) {r=true}
  n24 = NamedRerouteUsage()
  n25 = Multiply(
   A=Multiply(A=Add(A=@n13, B=@n18), B=Sine(None=@n23))
   B=Multiply(A=ScalarParameter() {defaultValue=0.025, group=Raptor Mouth, parameterName=Noise_WaveAmplitude}, B=@n24)
  )
  n26 = TextureCoordinate()
  n27 = ComponentMask(None=Constant3Vector() {desc=StartingPoint, constant=<rgba 0.055 0.475 0 0>}) {g=true, r=true}
  n28 = Divide(A=Distance(A=ComponentMask(None=@n26) {r=true}, B=ComponentMask(None=@n27) {r=true}), B=Constant() {desc=LineLength, r=0.2})
  n29 = ConstantBiasScale(None=Cosine(None=@n28))
  n30 = Subtract(A=@n26, B=@n27)
  n31 = TextureSample(UVs=MaterialFunctionCall(UVs=TextureCoordinate() {uTiling=0.4, vTiling=0.4}, Texture Scale=Constant() {r=25}) {materialFunction=&ScaleUVsByCenter}.UVs) {samplerType=SAMPLERTYPE_LinearColor, texture=/Game/Global/Textures/Default/Noise/TilingNoise06_N_uncompressed}
  n32 = Multiply(A=Subtract(A=ComponentMask(None=@n31.RGB) {g=true, r=true}) {constB=0.5}) {constB=0.15}
  n33 = TextureCoordinate()
  n34 = Time()
  n35 = Sine(None=@n34) {period=2}
  n36 = Sine(None=@n34) {period=3.45}
  n37 = Clamp(None=Ceil(None=Add(A=@n35, B=@n36))) {desc=Scale Range, maxDefault=0.4, minDefault=0.2}
  n38 = Add(
   A=Add(A=TextureCoordinate(), B=Constant2Vector() {desc=Cordinates, g=-0.95, r=1})
   B=MaterialFunctionCall(UVs=Add(A=@n32, B=@n33) {constB=0}, Texture Scale=Saturate(None=@n37)) {materialFunction=&ScaleUVsByCenter}.UVs
  )
  n39 = Add(
   A=Multiply(A=OneMinus(None=@n29), B=ScalarParameter() {defaultValue=-0.2, group=Raptor Mouth, parameterName=Smile_BendDistance})
   B=Add(A=ComponentMask(None=@n30) {g=true}, B=ComponentMask(None=@n38) {g=true})
  )
  n40 = Abs(None=Divide(A=Add(A=@n25, B=@n39), B=ScalarParameter() {defaultValue=0.04, group=Raptor Mouth, parameterName=LineWidth}))
  n41 = Multiply(A=OneMinus(None=Saturate(None=@n40)), B=@n24)
  n42 = OneMinus(None=NamedRerouteUsage())
  n43 = Add(
   A=StaticSwitchParameter(True=LinearInterpolate(A=@n41, Alpha=@n42) {constB=0}, False=@n41) {defaultValue=true, group=Raptor Mouth, parameterName=isBehindTeeth?}
   B=NamedRerouteUsage()
  )
  NamedRerouteDeclaration(None=Multiply(A=NamedRerouteUsage(), B=@n43)) {name=Diffuse_Output, nodeColor=<rgba 1 0 0.7734 1>}
  NamedRerouteDeclaration(None=@n43) {name=Mask_Output, nodeColor=<rgba 0.875 0 1 1>}
  NamedRerouteDeclaration(None=OneMinus(None=Saturate(None=@n28))) {name=LineGradient, nodeColor=<rgba 1 0 0.1406 1>}
  n44 = Multiply(A=Time(), B=ScalarParameter() {defaultValue=2, group=Raptor Mouth, parameterName=Flash Speed})
  n45 = Sine(None=@n44) {period=1.27}
  n46 = Constant() {r=0.2}
  n47 = Sine(None=@n44) {period=2.5}
  n48 = Constant() {r=0.35}
  n49 = Clamp(None=Add(A=Multiply(A=@n45, B=@n46), B=Multiply(A=@n47, B=@n48))) {maxDefault=0.5}
  n50 = ConstantBiasScale(None=Sine(None=@n44) {period=0.07})
  n51 = Saturate(None=SphereMask(
    A=TextureCoordinate()
    B=Constant2Vector() {g=0.515, r=0.153}
    Radius=ScalarParameter() {defaultValue=25, group=Raptor Mouth, parameterName=Mouth Gradient Radius}
    Hardness=ScalarParameter() {group=Raptor Mouth, parameterName=Mouth Gradient Hardness}
   ) {attenuationRadius=25, hardnessPercent=0})
  n52 = Add(
   A=Multiply(A=ScalarParameter() {defaultValue=1, group=Raptor Mouth, parameterName=Flash Intensity}, B=Multiply(A=@n49, B=@n50))
   B=Multiply(A=ScalarParameter() {defaultValue=80, group=Raptor Mouth, parameterName=Emissive Intensity Raptor}, B=OneMinus(None=@n51))
  )
  n53 = NamedRerouteUsage()
  n54 = TextureSampleParameter2D(UVs=Add(A=TextureCoordinate(), B=Constant2Vector() {desc=Cordinates, g=0.01, r=1})) {
   parameterName=MaskFX
   samplerType=SAMPLERTYPE_Masks
   texture=&T_F_MED_Raptor_Purple_Mask_FX
  }
  n55 = TextureSampleParameter2D(UVs=Add(A=TextureCoordinate(), B=Constant2Vector() {g=-0.01, r=1})) {
   parameterName=MaskFX
   samplerType=SAMPLERTYPE_Masks
   texture=&T_F_MED_Raptor_Purple_Mask_FX
  }
  n56 = Constant() {r=2}
  n57 = Multiply(A=Subtract(A=OneMinus(None=@n54.B), B=@n55.B), B=@n56) {constB=5, desc=bottom edge}
  n58 = TextureSampleParameter2D(UVs=Add(A=TextureCoordinate(), B=Constant2Vector() {desc=Cordinates, g=0.01, r=1})) {
   parameterName=MaskFX
   samplerType=SAMPLERTYPE_Masks
   texture=&T_F_MED_Raptor_Purple_Mask_FX
  }
  n59 = Multiply(A=Subtract(A=OneMinus(None=@n58.B), B=@n55.B), B=@n56) {constB=5, desc=top edge}
  n60 = Saturate(None=Add(A=SmoothStep(Min=@n57), B=SmoothStep(Min=@n59)))
  n61 = Multiply(A=@n52, B=Multiply(A=OneMinus(None=@n60), B=NamedRerouteUsage()))
  NamedRerouteDeclaration(None=Add(A=Multiply(A=@n52, B=@n53), B=Multiply(A=@n61) {constB=0.25})) {name=FX_Flash, nodeColor=<rgba 0.1719 0 1 1>}
 ]}
